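# Physics asset for the Cranium backpack mesh, giving capsule bodies to the pack and its dangling chain and strap bones.
# A three-link chain and a five-segment strap, starting at strap_base, each form a series of constraints rooted on the pack body.
# Chain joints share uniform angle limit values, while strap swing limit values widen toward the tip.
# Twist is locked between strap segments, but not where strap_base joins the pack.
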PhysicsAsset /Game/Accessories/FORT_Backpacks/Backpack_M_MED_Cranium/Meshes/M_MED_Cranium_Pack_Physics
bodies {
 chain1 [{shapeType=Capsule, length=0.505, radius=0.505, center=<0 0.0016 -0.5175>, rotation=<rot -0.1742 -90 0>}]
 chain2 [{shapeType=Capsule, length=0.505, radius=0.505, center=<0.0013 0 -0.5175>, rotation=<rot 0 -90 -0.1455>}]
 chain3 [{shapeType=Capsule, length=0.505, radius=0.505, center=<0 0.0672 -0.5131>, rotation=<rot -7.4574 -89.9999 0>}]
 pack [{shapeType=Capsule, length=30.514, radius=9.8523, center=<2.5599 10.245 1.1695>, rotation=<rot -18.8257 -89.9188 -0.2316>}]
 strap1 [{shapeType=Capsule, length=1.9565, radius=0.9118, center=<2.1331 0 0>, rotation=<rot 84.9997 89.9999 89.9999>}]
 strap2 [{shapeType=Capsule, length=2.1414, radius=0.875, center=<2.2994 0 0>, rotation=<rot 84.9994 89.9996 89.9996>}]
 strap3 [{shapeType=Capsule, length=1.7417, radius=0.8069, center=<2.3657 0 0>, rotation=<rot 84.9994 89.9996 89.9996>}]
 strap4 [{shapeType=Capsule, length=1.255, radius=0.505, center=<1.497 0 0>, rotation=<rot 84.9994 89.9996 89.9996>}]
 strap_base [{shapeType=Capsule, length=1.005, radius=0.755, center=<1.0258 0 0>, rotation=<rot 84.9994 89.9996 89.9996>}]
}
constraints [
 {bone1Name=chain2, bone2Name=chain1, swing1LimitDegrees=25, swing2LimitDegrees=25, twistLimitDegrees=25}
 {bone1Name=chain3, bone2Name=chain2, swing1LimitDegrees=25, swing2LimitDegrees=25, twistLimitDegrees=25}
 {bone1Name=strap1, bone2Name=strap_base, swing1LimitDegrees=6, swing2LimitDegrees=6, twistLimitDegrees=5, twistMotion=Locked}
 {bone1Name=strap2, bone2Name=strap1, swing1LimitDegrees=9, swing2LimitDegrees=9, twistLimitDegrees=5, twistMotion=Locked}
 {bone1Name=strap3, bone2Name=strap2, swing1LimitDegrees=12, swing2LimitDegrees=12, twistLimitDegrees=5, twistMotion=Locked}
 {bone1Name=strap4, bone2Name=strap3, swing1LimitDegrees=15, swing2LimitDegrees=15, twistLimitDegrees=5, twistMotion=Locked}
 {bone1Name=chain1, bone2Name=pack, swing1LimitDegrees=25, swing2LimitDegrees=25, twistLimitDegrees=25}
 {bone1Name=strap_base, bone2Name=pack, swing1LimitDegrees=6, swing2LimitDegrees=6, twistLimitDegrees=5}
]
